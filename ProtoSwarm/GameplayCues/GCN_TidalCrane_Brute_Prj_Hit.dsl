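; Gameplay cue burst for the TidalCrane Brute projectile hit, handling impact effect orientation and a physics push.
; On burst, it rotates the first particle component to the hit normal and sets its FXScale parameter, if that component is valid.
; It then spawns an FN_RadialForce field actor at the target's root location, gated to non-mobile platforms outside PC performance mode with detail mode above 1.
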(fn OnBurst(EitherVFX-preferthisversion)
  (params (MyTarget :type Actor%20Object%20Reference) (Parameters :type Gameplay%20Cue%20Parameters%20Structure%20%28by%20ref%29) (ParticleComponents :type Array%20of%20FXSystem%20Component%20Object%20References) (AudioComponents :type Array%20of%20Audio%20Component%20Object%20References) (BurstCameraShakeInstance :type Camera%20Shake%20Base%20Object%20Reference) (BurstDecalInstance :type Decal%20Actor%20Object%20Reference))
  (bind (_returnvalue_bblockinghit _returnvalue_binitialoverlap _returnvalue_time _returnvalue_distance _returnvalue_location _returnvalue_impactpoint _returnvalue_normal) (Collision|BreakHitResult (Ability|GameplayCue|GetHitResult Parameters)))
  (bind _output (Utilities|Array|Get(acopy) ParticleComponents))
  (Utilities|IsValid _output
    (:Is%20Valid
      (Transformation|SetWorldRotation _output (Math|Conversions|RotationFromXVector _returnvalue_normal) false false)
      (Effects|Components|ParticleSystem|SetFloatParameter _output "FXScale" 1.0))
    (:Is%20Not%20Valid
      _))
  (if (Math|Boolean|ANDBoolean (not (Fortnite|IsMobilePlatform)) (not (Class|PlatformHelpersLibrary|IsPcPerfMode 0)) (> (Rendering|GetRenderingDetailMode) 1))
    (Game|SpawnActorfromClass 0 "/Game/Effects/Niagara/Blueprints/PhsyicsFields/FN_RadialForce.FN_RadialForce_C" (Math|Transform|MakeTransform (Transformation|GetWorldLocation (Variables|/Script/Engine.Actor|GetRootComponent MyTarget))) "Undefined" "MultiplyWithRoot" 0 3.0 0.6 1e+03 1.2 0.05 0)
    (return)))
