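; Gameplay cue notify for the Medium TidalCrane enemy's unburrow effect.
; On execute it runs the parent cue behavior, then shows a sound indicator at the target actor's location for a short duration, using the FortSoundIndicator variable.
; The burst handler returns false, and the begin play, overlap, tick and construction script events are empty.
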(event EventBeginPlay)

(event Collision|EventActorBeginOverlap
  (params (OtherActor :type Actor%20Object%20Reference)))

(event EventTick
  (params (DeltaSeconds :type Float%20%28single-precision%29)))

(fn OnBurst
  (params (MyTarget :type Actor%20Object%20Reference) (Parameters :type Gameplay%20Cue%20Parameters%20Structure%20%28by%20ref%29))
  (return false))

(fn OnExecute
  (params (MyTarget :type Actor%20Object%20Reference) (Parameters :type Gameplay%20Cue%20Parameters%20Structure%20%28by%20ref%29))
  (|CallParentFunction MyTarget Parameters)
  (SoundIndicator|ShowSoundIndicatorforDuration (Variables|Default|GetFortSoundIndicator) (Transformation|GetActorLocation MyTarget) 0 1.0 1.0 0)
  (return true))

(fn ConstructionScript)
